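# Looping spatial ambient cue for water bobbing around a fiberglass sailboat, built from a metal boat bobbing loop.
# A random node picks one of two differently pitched copies of that same wave.
# Nested looping envelopes slowly vary pitch and volume after a short random start delay.
# Capsule attenuation with listener focus and occlusion, per-owner concurrency and an exterior ambience bus send shape playback.
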SoundCue /Game/Sounds/Helios/Environment_Actors/SailBoat/Spot_Water_Fiberglass_SailBoat_Bobbing_Around_Loop_Cue
root=Delay_0
nodes [
 {id=Delay_0, children [Enveloper_3], properties {delayMax=0.5}}
 {
  id=Enveloper_0
  children [Random_0]
  properties {
   bLoop=true
   bLoopIndefinitely=true
   loopEnd=30
   pitchCurve {editorCurveData {keys [
      {arriveTangent=-0.0013, interpMode=RCIM_Cubic, leaveTangent=-0.0013, time=0}
      <at 15 0.98 RCIM_Cubic>
      {arriveTangent=0.0013, interpMode=RCIM_Cubic, leaveTangent=0.0013, time=30}
     ]}}
   volumeCurve {editorCurveData {keys [<at 0 1>]}}
  }
 }
 {
  id=Enveloper_3
  children [Enveloper_0]
  properties {
   bLoop=true
   bLoopIndefinitely=true
   loopEnd=10
   pitchCurve {editorCurveData {keys [{time=0}]}}
   volumeCurve {editorCurveData {keys [
      {arriveTangent=-0.01, interpMode=RCIM_Cubic, time=0, value=1}
      <at 5 0.7 RCIM_Cubic>
      {arriveTangent=-0.03, interpMode=RCIM_Cubic, leaveTangent=-0.03, time=10, value=1}
      <at 15 0.4 RCIM_Cubic>
      <at 20 1 RCIM_Cubic>
     ]}}
  }
 }
 {id=Modulator_0, children [WavePlayer_0], properties {pitchMax=0.9, pitchMin=0.8}}
 {id=Modulator_1, children [WavePlayer_1], properties {pitchMax=1.1, pitchMin=1}}
 {id=Random_0, children [Modulator_0,Modulator_1], properties {weights [1,1]}}
 {
  id=WavePlayer_0
  properties {bLooping=true, soundWaveAssetPtr=/Game/Sounds/Asteria/Environment_Actors/Boat/Boat_Metal/Spot_Water_Metal_Boat_Bobbing_Around_Loop}
 }
 {
  id=WavePlayer_1
  properties {bLooping=true, soundWaveAssetPtr=&Spot_Water_Metal_Boat_Bobbing_Around_Loop}
 }
]
properties {
 bOverrideAttenuation=true
 bOverrideConcurrency=true
 maxDistance=1340
 soundClassObject=&Fort_Ambient_Outdoors_Athena
 subtitlePriority=5000
 volumeMultiplier=0.451
 attenuationOverrides {
  attenuationShape=Capsule
  bEnableFocusInterpolation=true
  bEnableListenerFocus=true
  dBAttenuationAtMax=-40
  falloffDistance=750
  focusAzimuth=15
  focusVolumeAttenuation=1.2
  nonFocusAzimuth=45
  nonFocusVolumeAttenuation=0.5
  occlusionInterpolationTime=1
  occlusionLowPassFilterFrequency=10000
  occlusionTraceChannel=ECC_GameTraceChannel11
  occlusionVolumeAttenuation=0.5
  attenuationShapeExtents {x=590, z=90}
 }
 busSends [{audioBus=null, minSendLevel=0.6, sendLevel=1, soundSourceBus=&AmbientAudio_Exterior_SourceBus}]
 concurrencyOverrides {bLimitToOwner=true, maxCount=2}
}
